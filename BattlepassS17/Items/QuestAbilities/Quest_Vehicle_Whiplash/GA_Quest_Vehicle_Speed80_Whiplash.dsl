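; Battle pass quest ability for the Whiplash vehicle that credits a quest when the player's car reaches a required forward speed.
; On a vehicle-entered event, it tag-checks the car and starts a looping one-second timer.
; Each check debug-prints the km/h speed and, at the threshold, sends a complex custom stat event to the quest manager and clears the timer.
; A vehicle-exit event clears the timer and ends the ability.
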(event Custom|Check%20Speed
  (bind _asfort_athena_vehicle (Utilities|Casting|CastToFortAthenaVehicle (Vehicle|GetVehicleInterface (Variables|GAVars|Data|GetOwningPlayerPawn)))
    (:then
      (Development|PrintString (Utilities|String|ToString(Float) (Vehicle|GetForwardSpeedKmh _asfort_athena_vehicle)))
      (if (>= (Vehicle|GetForwardSpeedKmh _asfort_athena_vehicle) (Variables|Default|GetRequired%20Speed))
        (Quest|SendComplexCustomStatEvent (Variables|GAVars|Data|GetOwningPlayerQuestManager) 0 (Variables|Default|GetT_Quest) "(GameplayTags=)" 1)
        (Utilities|Time|ClearandInvalidateTimerbyHandle (Variables|Default|GetTimer%20Handle))))
    (:CastFailed
      _))
  (Ability|EndAbility))

(event Ability|EventActivateAbilityFromEvent
  (params (EventData :type Gameplay%20Event%20Data%20Structure%20%28by%20ref%29))
  (bind _timer_handle (Variables|Default|GetTimer%20Handle))
  (|CallParentFunction EventData)
  (switch Utilities|FlowControl|Switch|SwitchonGameplayTag (Utilities|Struct|BreakGameplayEventData EventData)
    (:Event.Vehicle.Entered
      (if (GameplayTags|TagContainer|CompareTagAssetInterfacetoOtherTags (Variables|Default|GetT_Car) (Vehicle|GetVehicleAbilitySystemComponent (Vehicle|GetVehicleInterface (Variables|GAVars|Data|GetOwningPlayerPawn))))
        (bind _returnvalue (Utilities|Time|SetTimerbyEvent (delegate Custom|Check%20Speed) 1.0 true))
        (Variables|Default|SetTimer%20Handle _returnvalue)
        (else
          (Ability|EndAbility))))
    (:Event.Vehicle.Exit
      (if (Utilities|Time|IsValidTimerHandle _timer_handle)
        (Utilities|Time|ClearandInvalidateTimerbyHandle _timer_handle)
        (else
          _))
      (Ability|EndAbility))))
